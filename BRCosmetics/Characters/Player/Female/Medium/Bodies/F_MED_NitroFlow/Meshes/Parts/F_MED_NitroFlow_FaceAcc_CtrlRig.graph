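# Control Rig for the F_MED_NitroFlow female face accessory part that rotates the right hair chain as the head moves.
# A spherical pose reader measures head orientation relative to spine_05 and outputs a falloff-based weight.
# That weight blends a fixed Euler rotation onto dyn_hair_r_0, the chain's root bone, as a pose-driven corrective.
# The left and back hair chains, bang and facial bones are in the hierarchy but are not driven.
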ControlRigBlueprint /BRCosmetics/Characters/Player/Female/Medium/Bodies/F_MED_NitroFlow/Meshes/Parts/F_MED_NitroFlow_FaceAcc_CtrlRig
graphs [
 RigVMFunctionLibrary {links [], nodes []}
 RigVMModel {nodes [
   RigUnit_BeginExecution = &RigUnit_BeginExecution()
   n1 = &RigUnit_SphericalPoseReader(
    ActiveRegionScaleFactors=(PositiveWidth=1,NegativeWidth=1,PositiveHeight=1,NegativeHeight=1)
    ActiveRegionSize=0.1
    Debug=(bDrawDebug=false,DebugScale=25,DebugSegments=20,DebugThickness=0.25)
    DriverAxis=(X=0,Y=1,Z=0)
    DriverItem=(Type=Bone,Name="head")
    FalloffRegionScaleFactors=(PositiveWidth=1,NegativeWidth=1,PositiveHeight=1,NegativeHeight=1)
    FalloffSize=0.2
    OptionalParentItem=(Type=Bone,Name="spine_05")
    RotationOffset=(X=10,Y=0,Z=0)
    ExecutePin=@RigUnit_BeginExecution.ExecutePin
   )
   > &RigUnit_ModifyTransforms(
    ItemToModify=((Item=(Type=Bone,Name="dyn_hair_r_0"),Transform=(Rotation=(X=0,Y=0,Z=0,W=1),Translation=(X=0,Y=0,Z=0),Scale3D=(X=1,Y=1,Z=1))))
    ItemToModify.0.Transform.Rotation=&RigVMFunction_MathQuaternionFromEuler(Euler=(X=0,Y=0,Z=30)).Result
    Weight=@n1.OutputParam
   )
  ]}
]
hierarchy [
 root {loc=<0 0 0>, rot=<0 0 0 1>}
 pelvis {parent=root, loc=<0 -0.5301 98.0571>, rot=<0.0127 0.707 -0.0127 -0.707>}
 spine_01 {parent=pelvis, loc=<6.1208 -0.9328 0>, rot=<0 0 0.034 -0.9994>}
 spine_02 {parent=spine_01, loc=<6.5996 -0.2095 0>, rot=<0 0 0.0238 -0.9997>}
 spine_03 {parent=spine_02, loc=<6.4914 -0.1237 0>, rot=<0 0 0.0149 -0.9999>}
 spine_04 {parent=spine_03, loc=<6.9598 0.0794 0>, rot=<0 0 -0.0132 -0.9999>}
 spine_05 {parent=spine_04, loc=<6.9987 0 0>, rot=<0 0 -0.0015 -1>}
 neck_01 {parent=spine_05, loc=<11.3138 0.1529 0>, rot=<0 0 -0.2003 -0.9797>}
 neck_02 {parent=neck_01, loc=<4.5627 0 0>, rot=<0 0 0.0504 -0.9987>}
 head {parent=neck_02, loc=<5.3914 0 0>, rot=<0 0 0.115 -0.9934>}
 dyn_hair_b_0 {parent=head, loc=<15.2978 -11.979 0>, rot=<0.0921 -0.9957 0 0>}
 dyn_hair_b_1 {parent=dyn_hair_b_0, loc=<7.7441 0.9697 0>, rot=<0 0 -0.0829 -0.9966>}
 dyn_hair_b_2 {parent=dyn_hair_b_1, loc=<9.3376 0 0>, rot=<0 0 -0.0018 -1>}
 dyn_hair_b_3 {parent=dyn_hair_b_2, loc=<9.6271 0 0>, rot=<0 0 -0.0137 -0.9999>}
 dyn_hair_b_4 {parent=dyn_hair_b_3, loc=<9.0349 0 0>, rot=<0 0 0 1>}
 dyn_hair_l_0 {parent=head, loc=<6.8906 3.6067 -6.7374>, rot=<-0.024 -0.9997 0 0>}
 dyn_hair_l_1 {parent=dyn_hair_l_0, loc=<5.7817 0 0>, rot=<0 0 0.0225 -0.9997>}
 dyn_hair_l_2 {parent=dyn_hair_l_1, loc=<6.7926 0 0>, rot=<0 0 -0.0915 -0.9958>}
 dyn_hair_l_3 {parent=dyn_hair_l_2, loc=<4.7026 0 0>, rot=<0 0 -0.0571 -0.9984>}
 dyn_hair_l_4 {parent=dyn_hair_l_3, loc=<4.7513 0 0>, rot=<0 0 -0.0861 -0.9963>}
 dyn_hair_r_0 {parent=head, loc=<6.8902 3.6067 6.7374>, rot=<-0.024 -0.9997 0 0>}
 dyn_hair_r_1 {parent=dyn_hair_r_0, loc=<5.7817 0 0>, rot=<0 0 0.0225 -0.9997>}
 dyn_hair_r_2 {parent=dyn_hair_r_1, loc=<6.7926 0 0>, rot=<0 0 -0.0915 -0.9958>}
 dyn_hair_r_3 {parent=dyn_hair_r_2, loc=<4.7026 0 0>, rot=<0 0 -0.0571 -0.9984>}
 dyn_hair_r_4 {parent=dyn_hair_r_3, loc=<4.7513 0 0>, rot=<0 0 -0.0861 -0.9963>}
 FACIAL_C_FacialRoot {parent=head, loc=<2.0291 0.35 0>, rot=<0.5021 -0.4979 -0.5021 -0.4979>}
 FACIAL_C_Forehead {parent=FACIAL_C_FacialRoot, loc=<0 -2.7434 7.9054>, rot=<-0.0436 0 0 -0.999>}
 FACIAL_L_ForeheadIn {parent=FACIAL_C_FacialRoot, loc=<1.8038 -2.9023 7.7398>, rot=<-0.0609 -0.0696 0.0043 -0.9957>}
 FACIAL_R_ForeheadIn {parent=FACIAL_C_FacialRoot, loc=<-1.8038 -2.9023 7.7398>, rot=<-0.0609 0.0696 -0.0043 -0.9957>}
 FACIAL_L_CheekInner {parent=FACIAL_C_FacialRoot, loc=<2.119 1.0415 7.0308>, rot=<-0.1233 -0.088 -0.1972 -0.9686>}
 FACIAL_R_CheekInner {parent=FACIAL_C_FacialRoot, loc=<-2.119 1.0415 7.0308>, rot=<-0.1233 0.088 0.1972 -0.9686>}
 FACIAL_L_CheekOuter {parent=FACIAL_C_FacialRoot, loc=<3.8798 1.1439 5.1605>, rot=<0.0551 -0.377 0.1003 -0.9191>}
 FACIAL_R_CheekOuter {parent=FACIAL_C_FacialRoot, loc=<-3.8798 1.1439 5.1605>, rot=<0.0551 0.377 -0.1003 -0.9191>}
 FACIAL_L_NasolabialBulge {parent=FACIAL_C_FacialRoot, loc=<2.9275 4.4016 5.1677>, rot=<0.0579 -0.3774 0.0238 -0.9239>}
 FACIAL_R_NasolabialBulge {parent=FACIAL_C_FacialRoot, loc=<-2.9275 4.4016 5.1677>, rot=<0.0579 0.3774 -0.0238 -0.9239>}
 FACIAL_L_NasolabialFurrow {parent=FACIAL_C_FacialRoot, loc=<2.8656 3.2264 7.632>, rot=<-0.0333 -0.1258 -0.3638 -0.9223>}
 FACIAL_R_NasolabialFurrow {parent=FACIAL_C_FacialRoot, loc=<-2.8656 3.2264 7.632>, rot=<-0.0333 0.1258 0.3638 -0.9223>}
 FACIAL_L_CheekLower {parent=FACIAL_C_FacialRoot, loc=<3.3041 4.1839 4.4495>, rot=<0.0345 -0.427 0.0577 -0.9018>}
 FACIAL_R_CheekLower {parent=FACIAL_C_FacialRoot, loc=<-3.3041 4.1839 4.4495>, rot=<0.0345 0.427 -0.0577 -0.9018>}
 FACIAL_C_Nose {parent=FACIAL_C_FacialRoot, loc=<0 0.5914 8.496>, rot=<-0.1305 0 0 -0.9914>}
 FACIAL_L_Nostril {parent=FACIAL_C_Nose, loc=<0.961 1.9018 -0.3439>, rot=<0.0173 0.1305 0.0023 -0.9913>}
 FACIAL_R_Nostril {parent=FACIAL_C_Nose, loc=<-0.961 1.9018 -0.3439>, rot=<0.0173 -0.1305 -0.0023 -0.9913>}
 FACIAL_C_NoseTip {parent=FACIAL_C_Nose, loc=<0 1.9588 0.2589>, rot=<0.0872 0 0 -0.9962>}
 FACIAL_C_MouthUpper {parent=FACIAL_C_FacialRoot, loc=<0 4.3524 6.2917>, rot=<0 0 0 -1>}
 FACIAL_C_LipUpper {parent=FACIAL_C_MouthUpper, loc=<0 0.2631 2.2071>, rot=<-0.0872 0 0 -0.9962>}
 FACIAL_L_LipUpper {parent=FACIAL_C_MouthUpper, loc=<0.7741 0.2175 2.0901>, rot=<-0.0778 -0.1601 -0.0379 -0.9833>}
 FACIAL_R_LipUpper {parent=FACIAL_C_MouthUpper, loc=<-0.7741 0.2175 2.0901>, rot=<-0.0778 0.1601 0.0379 -0.9833>}
 FACIAL_L_LipUpperOuter {parent=FACIAL_C_MouthUpper, loc=<1.4147 0.421 1.5684>, rot=<-0.0983 -0.2294 -0.0983 -0.9634>}
 FACIAL_R_LipUpperOuter {parent=FACIAL_C_MouthUpper, loc=<-1.4147 0.421 1.5684>, rot=<-0.0983 0.2294 0.0983 -0.9634>}
 FACIAL_L_LipCorner {parent=FACIAL_C_MouthUpper, loc=<2.0501 0.7833 0.7397>, rot=<0 -0.2079 0 -0.9781>}
 FACIAL_R_LipCorner {parent=FACIAL_C_MouthUpper, loc=<-2.0501 0.7833 0.7397>, rot=<0 0.2079 0 -0.9781>}
 FACIAL_L_Masseter {parent=FACIAL_C_FacialRoot, loc=<5.9851 1.6116 3.0263>, rot=<0.0693 0.1129 0.0079 -0.9912>}
 FACIAL_R_Masseter {parent=FACIAL_C_FacialRoot, loc=<-5.9851 1.6116 3.0263>, rot=<0.0693 -0.1129 -0.0079 -0.9912>}
 FACIAL_C_UnderChin {parent=FACIAL_C_FacialRoot, loc=<0 8.2192 3.558>, rot=<0.5 0 0 -0.866>}
 FACIAL_L_UnderChin {parent=FACIAL_C_FacialRoot, loc=<2.8841 6.5434 2.3167>, rot=<0.4025 -0.355 0.0823 -0.8398>}
 FACIAL_R_UnderChin {parent=FACIAL_C_FacialRoot, loc=<-2.8841 6.5434 2.3167>, rot=<0.4025 0.355 -0.0823 -0.8398>}
 FACIAL_C_LowerLipRotation {parent=FACIAL_C_FacialRoot, loc=<0 1.8652 1.1847>, rot=<0.342 0 0 -0.9397>}
 FACIAL_C_MouthLower {parent=FACIAL_C_LowerLipRotation, loc=<0 -1.3774 5.511>, rot=<-0.342 0 0 -0.9397>}
 FACIAL_C_LipLower {parent=FACIAL_C_MouthLower, loc=<0 1.2786 1.9617>, rot=<0.0872 0 0 -0.9962>}
 FACIAL_L_LipLower {parent=FACIAL_C_MouthLower, loc=<0.6746 1.2446 1.792>, rot=<0.0927 -0.1295 0.0273 -0.9869>}
 FACIAL_R_LipLower {parent=FACIAL_C_MouthLower, loc=<-0.6746 1.2446 1.792>, rot=<0.0927 0.1295 -0.0273 -0.9869>}
 FACIAL_L_LipLowerOuter {parent=FACIAL_C_MouthLower, loc=<1.3355 1.1471 1.4186>, rot=<0.0913 -0.2322 0.0318 -0.9679>}
 FACIAL_R_LipLowerOuter {parent=FACIAL_C_MouthLower, loc=<-1.3356 1.1471 1.4186>, rot=<0.0913 0.2322 -0.0318 -0.9679>}
 FACIAL_C_Jaw {parent=FACIAL_C_FacialRoot, loc=<0 1.8652 1.1847>, rot=<0.342 0 0 -0.9397>}
 FACIAL_C_Jawline {parent=FACIAL_C_Jaw, loc=<0 2.3254 7.1604>, rot=<0.2588 0 0 -0.9659>}
 FACIAL_L_ChinSide {parent=FACIAL_C_Jaw, loc=<2.3977 0.1556 7.1401>, rot=<-0.2207 -0.3739 0.05 -0.8994>}
 FACIAL_R_ChinSide {parent=FACIAL_C_Jaw, loc=<-2.3977 0.1556 7.1401>, rot=<-0.2207 0.3739 -0.05 -0.8994>}
 FACIAL_C_Chin {parent=FACIAL_C_Jaw, loc=<0 0.8392 7.6471>, rot=<-0.2588 0 0 -0.9659>}
 FACIAL_L_Ear {parent=FACIAL_C_FacialRoot, loc=<6.5857 -0.2921 -0.4803>, rot=<-0.0262 -0.0007 -0.0262 -0.9993>}
 FACIAL_R_Ear {parent=FACIAL_C_FacialRoot, loc=<-6.5857 -0.2921 -0.4803>, rot=<-0.0262 0.0007 0.0262 -0.9993>}
 FACIAL_L_Jawline {parent=FACIAL_C_FacialRoot, loc=<4.2305 4.9049 3.9524>, rot=<0.2092 0.1578 -0.0547 -0.9635>}
 FACIAL_R_Jawline {parent=FACIAL_C_FacialRoot, loc=<-4.2305 4.9049 3.9524>, rot=<0.2092 -0.1578 0.0547 -0.9635>}
 FACIAL_C_NoseBridge {parent=FACIAL_C_FacialRoot, loc=<0 -1.3113 8.7611>, rot=<0.0436 0 0 -0.999>}
 dyn_bang {parent=head, loc=<12.8619 6.3553 0>, rot=<-0.1862 -0.6822 0.1862 -0.6822>}
 extra_lod_NoseTip {parent=head, loc=<-0.4637 9.5827 0>, rot=<0.4799 -0.5194 -0.4799 -0.5194>}
 clavicle_l {parent=spine_05, loc=<5.4831 -0.1053 -3.2946>, rot=<0.0465 -0.7289 -0.009 -0.6829>}
 upperarm_l {parent=clavicle_l, loc=<9.8658 -1.2136 -0.4105>, rot=<0.0164 -0.3245 -0.0269 -0.9454>}
 clavicle_r {parent=spine_05, loc=<5.4831 -0.1053 3.2946>, rot=<0.7289 0.0465 -0.6829 0.009>}
 upperarm_r {parent=clavicle_r, loc=<-9.8658 1.2136 0.4105>, rot=<0.0164 -0.3245 -0.0269 -0.9454>}
]
