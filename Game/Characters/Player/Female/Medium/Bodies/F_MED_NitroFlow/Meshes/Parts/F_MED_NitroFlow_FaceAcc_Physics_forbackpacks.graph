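# Physics asset for the NitroFlow female medium character's face accessory, simulating hair; the name marks it as a backpack variant.
# Capsules and a head sphere cover upper spine, neck, head, clavicles and upper arms as body shapes.
# Back, left and right five-bone hair chains plus a bang bone hang from the head through angle-limited constraints.
# Side chains loosen toward their tips, the back chain stays tight, and only the right upperarm-to-clavicle joint swings freely.
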PhysicsAsset /Game/Characters/Player/Female/Medium/Bodies/F_MED_NitroFlow/Meshes/Parts/F_MED_NitroFlow_FaceAcc_Physics_forbackpacks
bodies {
 clavicle_l [{
   shapeType=Capsule
   length=16.3363
   radius=5.4415
   shapeName=clavicle_r_capsule
   center=<3.1573 0 2.6037>
   rotation=<rot 84.8637 -0.0026 -0.0027>
  }]
 clavicle_r [{
   shapeType=Capsule
   length=16.3363
   radius=5.4415
   shapeName=clavicle_r_capsule
   center=<-3.2327 0 -3.4425>
   rotation=<rot 84.8637 -0.0026 -0.0027>
  }]
 dyn_bang [{shapeType=Capsule, length=5.916, radius=0.621, shapeName=dyn_bang_capsule, center=<0 0 -5.1159>}]
 dyn_hair_b_0 [{
   shapeType=Capsule
   length=0.8909
   radius=4.6705
   shapeName=dyn_hair_b_0_capsule
   center=<3.7152 -1.2453 0.1447>
   rotation=<rot 0.5927 -83.249 -88.1198>
  }]
 dyn_hair_b_1 [{
   shapeType=Capsule
   length=1.2756
   radius=4.5479
   shapeName=dyn_hair_b_1_capsule
   center=<3.4583 -1.3496 0.0089>
   rotation=<rot 3.5527 -91.2924 89.8116>
  }]
 dyn_hair_b_2 [{
   shapeType=Capsule
   length=2.4419
   radius=4.294
   shapeName=dyn_hair_b_2_capsule
   center=<3.958 -1.6462 0.2298>
   rotation=<rot 1.7642 -89.9567 91.6891>
  }]
 dyn_hair_b_3 [{
   shapeType=Capsule
   length=3.5541
   radius=3.9286
   shapeName=dyn_hair_b_3_capsule
   center=<2.5793 -2.1676 -0.044>
   rotation=<rot 0.2498 -89.9996 90.6195>
  }]
 dyn_hair_b_4 [{
   shapeType=Capsule
   length=4.7684
   radius=3.3256
   shapeName=dyn_hair_b_4_capsule
   center=<3.7146 -2.1208 -0.0234>
   rotation=<rot 0 90.9703 -89.3666>
  }]
 dyn_hair_l_0 [{
   shapeType=Capsule
   length=2.6067
   radius=1.7337
   shapeName=dyn_hair_l_0_capsule
   center=<1.6958 0.2701 1.0332>
   rotation=<rot -0.0133 -91.4311 89.4683>
  }]
 dyn_hair_l_1 [{
   shapeType=Capsule
   length=3.2805
   radius=1.7337
   shapeName=dyn_hair_l_1_capsule
   center=<2.34 0.0459 1.0223>
   rotation=<rot -0.3869 89.7075 -92.8934>
  }]
 dyn_hair_l_2 [{
   shapeType=Capsule
   length=1.3782
   radius=1.8058
   shapeName=dyn_hair_l_2_capsule
   center=<1.4441 0.2416 0.7448>
   rotation=<rot 1.3982 -88.9621 -90.4172>
  }]
 dyn_hair_l_3 [{
   shapeType=Capsule
   length=1.9143
   radius=1.8433
   shapeName=dyn_hair_l_3_capsule
   center=<1.6787 0.3398 0.7804>
   rotation=<rot 2.0694 -89.922 -90.01>
  }]
 dyn_hair_l_4 [{
   shapeType=Capsule
   length=3.6017
   radius=1.9192
   shapeName=dyn_hair_l_4_capsule
   center=<2.8385 0.2439 0.6883>
   rotation=<rot 1.1697 -94.1176 -90.9513>
  }]
 dyn_hair_r_0 [{
   shapeType=Capsule
   length=3.0864
   radius=1.254
   shapeName=dyn_hair_l_0_capsule
   center=<2.1455 -0.0013 -0.7353>
   rotation=<rot -0.0133 -91.4309 -90.7033>
  }]
 dyn_hair_r_1 [{
   shapeType=Capsule
   length=3.7213
   radius=1.293
   shapeName=dyn_hair_l_1_capsule
   center=<2.4876 -0.0675 -0.5759>
   rotation=<rot -0.4129 91.6888 90.7447>
  }]
 dyn_hair_r_2 [{
   shapeType=Capsule
   length=2.2986
   radius=1.2268
   shapeName=dyn_hair_l_2_capsule
   center=<1.4136 0.1561 -0.3057>
   rotation=<rot 1.3931 -89.619 89.5662>
  }]
 dyn_hair_r_3 [{
   shapeType=Capsule
   length=1.9251
   radius=1.3151
   shapeName=dyn_hair_l_3_capsule
   center=<1.5629 0.1339 -0.1418>
   rotation=<rot 2.066 -86.3774 90.1179>
  }]
 dyn_hair_r_4 [{
   shapeType=Capsule
   length=3.6017
   radius=1.9192
   shapeName=dyn_hair_l_4_capsule
   center=<2.8973 0.4195 0.059>
   rotation=<rot 1.2372 -89.9023 89.1385>
  }]
 head [
  {shapeType=Sphere, radius=9.1683, shapeName=head_sphere, center=<7.0731 0 0>}
  {shapeType=Capsule, length=6.1942, radius=6.0538, center=<1.6305 3.8295 0.0098>, rotation=<rot -0.0474 -68.4161 -89.9641>}
 ]
 neck_01 [{
   shapeType=Capsule
   length=11.0087
   radius=4.6572
   shapeName=neck_01_capsule
   center=<0.1587 0.7496 -0.0192>
   rotation=<rot 1.4468 -99.4449 89.7594>
  }]
 spine_03 [{shapeType=Capsule, length=10.3306, radius=10.9399, shapeName=spine_03_capsule, center=<-1.3035 5.2116 0>}]
 spine_04 [
  {shapeType=Capsule, length=24.0813, radius=8.8295, shapeName=spine_04_capsule, center=<6.1205 0.5202 -0.7118>}
  {shapeType=Capsule, length=10, radius=9.4062, shapeName=spine_04_capsule, center=<0.0074 4.5526 0>}
 ]
 spine_05 [{
   shapeType=Capsule
   length=17.3218
   radius=6.1989
   shapeName=spine_05_capsule
   center=<4.7766 -1.443 -0.4699>
   rotation=<rot 1.9516 177.2102 -179.9055>
  }]
 upperarm_l [{
   shapeType=Capsule
   length=20.6687
   radius=5.184
   shapeName=upperarm_l_capsule
   center=<13.5778 -0.6418 0.1717>
   rotation=<rot -86.8192 179.995 -179.9948>
  }]
 upperarm_r [{
   shapeType=Capsule
   length=21.8104
   radius=4.0424
   shapeName=upperarm_l_capsule
   center=<-13.5778 0.6418 -0.1717>
   rotation=<rot -86.8192 179.995 0.0052>
  }]
}
constraints [
 {
  bone1Name=upperarm_r
  bone2Name=clavicle_r
  swing1LimitDegrees=45
  swing1Motion=Free
  swing2LimitDegrees=45
  swing2Motion=Free
  twistLimitDegrees=45
  twistMotion=Free
 }
 {bone1Name=dyn_hair_b_0, bone2Name=head, swing1LimitDegrees=3, swing2LimitDegrees=3, twistLimitDegrees=5}
 {bone1Name=dyn_hair_b_1, bone2Name=dyn_hair_b_0, swing1LimitDegrees=5, swing2LimitDegrees=5, twistLimitDegrees=5}
 {bone1Name=dyn_hair_b_2, bone2Name=dyn_hair_b_1, swing1LimitDegrees=5, swing2LimitDegrees=5, twistLimitDegrees=5}
 {bone1Name=dyn_hair_b_3, bone2Name=dyn_hair_b_2, swing1LimitDegrees=5, swing2LimitDegrees=5, twistLimitDegrees=5}
 {bone1Name=dyn_hair_b_4, bone2Name=dyn_hair_b_3, swing1LimitDegrees=5, swing2LimitDegrees=5, twistLimitDegrees=5}
 {bone1Name=dyn_hair_l_0, bone2Name=head, swing1LimitDegrees=7, swing2LimitDegrees=7, twistLimitDegrees=5}
 {bone1Name=dyn_hair_l_1, bone2Name=dyn_hair_l_0, swing1LimitDegrees=13, swing2LimitDegrees=10, twistLimitDegrees=5}
 {bone1Name=dyn_hair_l_2, bone2Name=dyn_hair_l_1, swing1LimitDegrees=20, swing2LimitDegrees=20, twistLimitDegrees=5}
 {bone1Name=dyn_hair_l_3, bone2Name=dyn_hair_l_2, swing1LimitDegrees=20, swing2LimitDegrees=20, twistLimitDegrees=5}
 {bone1Name=dyn_hair_l_4, bone2Name=dyn_hair_l_3, swing1LimitDegrees=30, swing2LimitDegrees=30, twistLimitDegrees=5}
 {bone1Name=dyn_hair_r_0, bone2Name=head, swing1LimitDegrees=5, swing2LimitDegrees=5, twistLimitDegrees=5}
 {bone1Name=dyn_hair_r_1, bone2Name=dyn_hair_r_0, swing1LimitDegrees=15, swing2LimitDegrees=10, twistLimitDegrees=5}
 {bone1Name=dyn_hair_r_2, bone2Name=dyn_hair_r_1, swing1LimitDegrees=20, swing2LimitDegrees=20, twistLimitDegrees=5}
 {bone1Name=dyn_hair_r_3, bone2Name=dyn_hair_r_2, swing1LimitDegrees=20, swing2LimitDegrees=20, twistLimitDegrees=5}
 {bone1Name=dyn_hair_r_4, bone2Name=dyn_hair_r_3, swing1LimitDegrees=30, swing2LimitDegrees=30, twistLimitDegrees=5}
 {bone1Name=dyn_bang, bone2Name=head, swing1LimitDegrees=20, swing2LimitDegrees=10, twistLimitDegrees=1}
]
