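# Animation Blueprint for the F_MED_NitroFlow FaceAcc cosmetic part, adapting its hair dynamics to style and backbling.
# Copied parent and body poses feed parallel royal and default branches, each a Control Rig plus head-space RigidBody chain, chosen by a variant tag.
# Constraints and offsets then adjust the back hair bones before a backbling Control Rig.
# Updates set gravity override, backbling, front-end and LOD flags; customization load resets dynamics after a delay.
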AnimBlueprint /BRCosmetics/Characters/Player/Female/Medium/Bodies/F_MED_NitroFlow/Meshes/Parts/F_MED_NitroFlow_FaceAcc_AnimBP
target_skeleton=/Game/Characters/Player/Female/Medium/Bodies/F_MED_NitroFlow/Meshes/Parts/F_MED_NitroFlow_FaceAcc_Skeleton
graphs [
 AnimGraph {nodes [
   n1 = Control Rig(Source=Use cached pose 'begin'(), Alpha=Get ControlRigAlpha() -> (ControlRigAlpha:double)) {
    alphaBoolBlend {blendOption=Linear}
    controlRigAssetReference {blueprintRigClass=&/F_MED_NitroFlow_FaceAcc_CtrlRig.F_MED_NitroFlow_FaceAcc_CtrlRig_C}
   }
   n2 = RigidBody!(
    ComponentPose=Local To Component(LocalPose=@n1)
    bAlphaBoolEnabled:bool=Get NoBackbling() -> (NoBackbling:bool)
    OverrideWorldGravity:Vector=Get Gravity Override()
   ) -> (Pose) {
    alphaInputType=Bool
    bOverrideWorldGravity=true
    overridePhysicsAsset=/BRCosmetics/Characters/Player/Female/Medium/Bodies/F_MED_NitroFlow_Royal/Meshes/Parts/F_MED_NitroFlow_royal_FaceAcc_Physics
    simulationSpace=BaseBoneSpace
    alphaBoolBlend {blendOption=Linear}
    baseBoneRef {boneName=head}
    componentAppliedLinearAccClamp=<3000 3000 3000>
    componentLinearAccScale=<0.2 0.2 0.2>
    componentLinearVelScale=<0.2 0.2 0.2>
    overrideWorldGravity {z=-980}
    simSpaceSettings {maxAngularAcceleration=85, maxAngularVelocity=3, maxLinearAcceleration=300, maxLinearVelocity=500, velocityScaleZ=0, worldAlpha=1}
   }
   n3 = RigidBody!(ComponentPose=@n2, bAlphaBoolEnabled=Get backpack() -> (backpack:bool)) -> (Pose) {
    alphaInputType=Bool
    overridePhysicsAsset=/BRCosmetics/Characters/Player/Female/Medium/Bodies/F_MED_NitroFlow_Royal/Meshes/Parts/F_MED_NitroFlow_FaceAcc_Physics_backpack_royal
    simulationSpace=BaseBoneSpace
    alphaBoolBlend {blendOption=Linear}
    baseBoneRef {boneName=head}
    componentAppliedLinearAccClamp=<3000 3000 3000>
   }
   n4 = RigidBody(ComponentPose=@n3) {
    alphaInputType=Bool
    overridePhysicsAsset=/BRCosmetics/Characters/Player/Female/Medium/Bodies/F_MED_NitroFlow_Royal/Meshes/Parts/F_MED_NitroFlow_FaceAcc_Physics_backpack_FE_royal
    simulationSpace=BaseBoneSpace
    alphaBoolBlend {blendOption=Linear}
    baseBoneRef {boneName=head}
    componentAppliedLinearAccClamp=<3000 3000 3000>
   }
   n5 = Component To Local(ComponentPose=@n4)
   n6 = Control Rig(Source=Use cached pose 'begin'(), Alpha=Get ControlRigAlpha()) {
    alphaBoolBlend {blendOption=Linear}
    controlRigAssetReference {blueprintRigClass=&F_MED_NitroFlow_FaceAcc_CtrlRig_C}
   }
   n7 = RigidBody!(ComponentPose=Local To Component(LocalPose=@n6), bAlphaBoolEnabled=Get NoBackbling(), OverrideWorldGravity:Vector=Get Gravity Override()) -> (Pose) {
    alphaInputType=Bool
    bOverrideWorldGravity=true
    overridePhysicsAsset=/Game/Characters/Player/Female/Medium/Bodies/F_MED_NitroFlow/Meshes/Parts/F_MED_NitroFlow_FaceAcc_Physics
    simulationSpace=BaseBoneSpace
    alphaBoolBlend {blendOption=Linear}
    baseBoneRef {boneName=head}
    componentAppliedLinearAccClamp=<3000 3000 3000>
    componentLinearAccScale=<0.2 0.2 0.2>
    componentLinearVelScale=<0.2 0.2 0.2>
    overrideWorldGravity {z=-980}
    simSpaceSettings {maxAngularAcceleration=85, maxAngularVelocity=3, maxLinearAcceleration=300, maxLinearVelocity=500, velocityScaleZ=0, worldAlpha=1}
   }
   n8 = RigidBody!(ComponentPose=@n7, bAlphaBoolEnabled=Get backpack()) -> (Pose) {
    alphaInputType=Bool
    overridePhysicsAsset=/Game/Characters/Player/Female/Medium/Bodies/F_MED_NitroFlow/Meshes/Parts/F_MED_NitroFlow_FaceAcc_Physics_forbackpacks
    simulationSpace=BaseBoneSpace
    alphaBoolBlend {blendOption=Linear}
    baseBoneRef {boneName=head}
    componentAppliedLinearAccClamp=<3000 3000 3000>
   }
   n9 = RigidBody(ComponentPose=@n8, bAlphaBoolEnabled=Get backbling_FE() -> (backbling_FE:bool)) {
    alphaInputType=Bool
    overridePhysicsAsset=&/F_MED_NitroFlow_FaceAcc_Physics_forbackpack_FEs
    simulationSpace=BaseBoneSpace
    alphaBoolBlend {blendOption=Linear}
    baseBoneRef {boneName=head}
    componentAppliedLinearAccClamp=<3000 3000 3000>
   }
   n10 = Component To Local(ComponentPose=@n9)
   n11 = Get royal_style() -> (royal_style:bool)
   n12 = "Constraint - dyn_hair_b_1 (PP)"!(
    ComponentPose=Local To Component(LocalPose=Blend Poses by bool(BlendPose_0=@n5, BlendPose_1=@n10, bActiveValue=@n11) {blendType=Linear})
    Alpha=1
    ConstraintWeights_0=0.8
    ConstraintWeights_1:float=0.2
   ) -> (Pose) {
    alphaBoolBlend {blendOption=Linear}
    boneToModify {boneName=dyn_hair_b_1}
    constraintSetup [{targetBone {boneName=head}}, {targetBone {boneName=spine_05}}]
    constraintWeights [1,1]
   }
   n13 = "Constraint - dyn_hair_b_2 (PP)"!(ComponentPose=@n12, Alpha=1, ConstraintWeights_0=0.2, ConstraintWeights_1:float=0.8) -> (Pose) {
    alphaBoolBlend {blendOption=Linear}
    boneToModify {boneName=dyn_hair_b_2}
    constraintSetup [{targetBone {boneName=head}}, {targetBone {boneName=spine_05}}]
    constraintWeights [1,1]
   }
   n14 = "Constraint - dyn_hair_b_3 (PP)"!(ComponentPose=@n13, Alpha=1, ConstraintWeights_0=0, ConstraintWeights_1:float=1) -> (Pose) {
    alphaBoolBlend {blendOption=Linear}
    boneToModify {boneName=dyn_hair_b_3}
    constraintSetup [{targetBone {boneName=head}}, {targetBone {boneName=spine_05}}]
    constraintWeights [1,1]
   }
   n15 = "Constraint - dyn_hair_b_4 (PP)"!(ComponentPose=@n14, Alpha=1, ConstraintWeights_0=0, ConstraintWeights_1:float=1) -> (Pose) {
    alphaBoolBlend {blendOption=Linear}
    boneToModify {boneName=dyn_hair_b_4}
    constraintSetup [{targetBone {boneName=spine_04}}, {targetBone {boneName=spine_05}}]
    constraintWeights [1,1]
   }
   n16 = "Transform (Modify) Bone - Bone: dyn_hair_b_4"(ComponentPose=@n15, Alpha=1, Translation="0,1,0", Rotation="0,20,0", Scale="1,1,1") {
    rotationMode=BMM_Additive
    rotationSpace=BCS_BoneSpace
    scaleMode=BMM_Additive
    scaleSpace=BCS_BoneSpace
    translationMode=BMM_Additive
    translationSpace=BCS_BoneSpace
    alphaBoolBlend {blendOption=Linear}
    boneToModify {boneName=dyn_hair_b_4}
    rotation {yaw=20}
    translation {y=1}
   }
   n17 = "Transform (Modify) Bone - Bone: dyn_hair_b_2"(ComponentPose=@n16, Alpha=1, Translation="0,5,0", Rotation="0,-10,0", Scale="1,0.7,1") {
    rotationMode=BMM_Additive
    rotationSpace=BCS_BoneSpace
    scaleMode=BMM_Additive
    scaleSpace=BCS_BoneSpace
    translationMode=BMM_Additive
    translationSpace=BCS_BoneSpace
    alphaBoolBlend {blendOption=Linear}
    boneToModify {boneName=dyn_hair_b_2}
    rotation {yaw=-10}
    scale {y=0.7}
    translation {y=5}
   }
   n18 = "Transform (Modify) Bone - Bone: dyn_hair_b_3"(ComponentPose=@n17, Alpha=1, Translation="0,2,0", Rotation="0,0,0", Scale="1,0.5,1") {
    rotationMode=BMM_Additive
    rotationSpace=BCS_BoneSpace
    scaleMode=BMM_Additive
    scaleSpace=BCS_BoneSpace
    translationMode=BMM_Additive
    translationSpace=BCS_BoneSpace
    alphaBoolBlend {blendOption=Linear}
    boneToModify {boneName=dyn_hair_b_3}
    scale {y=0.5}
    translation {y=2}
   }
   n19 = "Transform (Modify) Bone - Bone: dyn_hair_b_4"(ComponentPose=@n18, Alpha=1, Translation="0,-1,0", Rotation="0,0,0", Scale="1,1,1") {
    rotationMode=BMM_Additive
    rotationSpace=BCS_BoneSpace
    scaleMode=BMM_Additive
    scaleSpace=BCS_BoneSpace
    translationMode=BMM_Additive
    translationSpace=BCS_BoneSpace
    alphaBoolBlend {blendOption=Linear}
    boneToModify {boneName=dyn_hair_b_4}
    translation {y=-1}
   }
   n20 = Control Rig!(Source=Component To Local(ComponentPose=@n19), Alpha=Get ControlRigAlpha(), RoyalStyleCtrl:bool=false) -> (Pose) {
    alphaBoolBlend {blendOption=Linear}
    controlRigAssetReference {blueprintRigClass=&/F_MED_NitroFlow_backbling_FaceAcc_CtrlRig.F_MED_NitroFlow_backbling_FaceAcc_CtrlRig_C}
   }
   Output Pose(Result=@n20) {layerGroup=None}
   n21 = Copy Pose From Mesh(SourceMeshComponent=Get body() -> (body:Skeletal Mesh Component Budgeted Object Reference)) {bUseAttachedParent=true}
   n22 = Copy Pose From Mesh() {bUseAttachedParent=true}
   n23 = "Transform (Modify) Bone - Bone: extra_lod_NoseTip"!(
    ComponentPose=Local To Component(LocalPose=Layered blend per bone(BasePose=@n22, BlendPoses_0=@n21) {layerSetup [{branchFilters [{boneName=upperarm_l}, {boneName=upperarm_r}]}]})
    bAlphaBoolEnabled:bool=Get lod1() -> (lod1:bool)
    Translation="0.035,0.529,-0.044"
    Rotation="0,0,0"
    Scale="1,1,1"
   ) -> (Pose) {
    alphaInputType=Bool
    translationMode=BMM_Additive
    translationSpace=BCS_ParentBoneSpace
    alphaBoolBlend {blendOption=Linear}
    boneToModify {boneName=extra_lod_NoseTip}
    translation=<0.035 0.529 -0.044>
   }
   Save cached pose 'begin'(Pose=Component To Local(ComponentPose=@n23)) {cacheName=begin}
  ]}
 EventGraph {nodes [
   Event BlueprintUpdateAnimation()
   > n1 = Sequence!(execute) -> (then_0, then_1, then_2:Exec)
   > n2 = Gravity Override State Function DEPRECATED(CharacterPartAnimInstance=Self-Reference(), GameTimeInSeconds=GetGameTimeInSeconds(), GravityOverrideParams=Get Gravity Override Params())
   > n3 = Branch(Condition=Get bIsFalling())
   > Set Gravity Override(Gravity Override="0, 0, 0")
   Set Gravity Override(execute=@n3.else, Gravity Override=@n2)
   n4 = CheckForBackbling()
   Set NoBackbling(execute=@n1.then_1, NoBackbling:bool="Not Equal (Boolean)"(A=@n4, B=true) -> (ReturnValue=~)) -> (Output_Get:bool)
   > n5 = Set HairPin(HairPin:bool=@n4) -> (Output_Get:bool)
   n6 = Get bIsInFrontEnd()
   > Set backpack(backpack:bool=AND Boolean(A="Not Equal (Boolean)"(A=@n6, B=true) -> (ReturnValue=~), B=@n4)) -> (Output_Get:bool)
   > Set backbling_FE(backbling_FE:bool=AND Boolean(A=@n4, B=@n6)) -> (Output_Get:bool)
   > Set ControlRigAlpha(ControlRigAlpha:double=SelectFloat(A=1, B=0, bPickA=@n4)) -> (Output_Get:double)
   n7 = Get royal_style()
   > Set Normal_Style(Normal_Style:bool=AND Boolean(A="Not Equal (Boolean)"(A=@n7, B=true) -> (ReturnValue=~), B=@n5)) -> (Output_Get:bool)
   Set lod1(execute=@n1.then_2, lod1:bool="To Boolean (Integer)"!(InInt:int=Get CurrentLOD()) -> (ReturnValue:bool)) -> (Output_Get:bool)
   n8 = Event OnCharacterCustomizationLoaded()
   > Set body(body:Skeletal Mesh Component Budgeted Object Reference="Get (a copy)"(Array:Array of Skeletal Mesh Component Budgeted Object References=@n8.CustomMeshParts, Dimension 1=1) -> (Output:Skeletal Mesh Component Budgeted Object Reference)) -> (Output_Get:Skeletal Mesh Component Budgeted Object Reference)
   > Set royal_style(royal_style:bool=CharacterItemDefinitionHasAnyMetaTag(Tags=Make Array("[0]"=(TagName="Cosmetics.Variant.Property.Stage2")))) -> (Output_Get:bool)
   > Delay(Duration=0.25)
   > ResetDynamics(InTeleportType=TeleportPhysics)
   > ResetDynamics(InTeleportType=ResetPhysics)
   "Not Equal (Boolean)"(A=Get backbling_FE()) -> (ReturnValue=~)
  ]}
]
variables [
 backbling_FE {type=boolean}
 backpack {type=boolean}
 body {type=SkeletalMeshComponentBudgeted}
 ControlRigAlpha {type=number}
 Gravity Override {type=Vector}
 Gravity Override Params {
  type=GravityOverrideParamsStruct
  default {
   onGround_Falling {gravityOverride=<0 0 0>}
   skydive_FloatingDown {gravityOverride=<0 -700 500>}
   skydive_GoingUp {gravityOverride=<0 -300 300>}
   skydive_Parachuting {gravityOverride=<0 0 -980>, windAmplitude=<200 -200 400>, windFrequency=<1.5 3 3>}
  }
 }
 HairPin {type=boolean}
 lod1 {type=boolean}
 NoBackbling {type=boolean}
 Normal_Style {type=boolean}
 Raw Output {type=number}
 royal_style {type=boolean}
]
